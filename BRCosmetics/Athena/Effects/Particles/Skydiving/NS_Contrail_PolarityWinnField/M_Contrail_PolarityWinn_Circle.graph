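# Translucent unlit, two-sided material for the PolarityWinn skydiving contrail, used on Niagara mesh particles and ribbons.
# Panned Perlin noise distorts UV gradients and scrolling sine lines, which drive a four-color emissive blend and opacity with optional Fresnel and depth fades.
# Scalar lerps adapt it to ribbon UVs or a bounding-box "hand" gradient, including noise-driven vertex offset along normals.
# Niagara dynamic parameters feed noise offset, distortion, audio reaction, line panning and gradient offset.
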Material /BRCosmetics/Athena/Effects/Particles/Skydiving/NS_Contrail_PolarityWinnField/M_Contrail_PolarityWinn_Circle
graph {
 properties {
  bAutomaticallySetUsageInEditor=false
  bUsedWithNiagaraMeshParticles=true
  bUsedWithNiagaraRibbons=true
  blendMode=BLEND_Translucent
  pixelDepthOffsetMode=PDOM_Legacy
  shadingModel=MSM_Unlit
  twoSided=true
  usedShadingModels=Unlit
 }
 nodes [
  n1 = ScalarParameter() {defaultValue=1, parameterName=DistortionTile1}
  n2 = ScalarParameter() {defaultValue=1, parameterName=DistortionTile2}
  n3 = Constant2Vector() {g=-0.4}
  n4 = Constant2Vector() {r=0.4}
  n5 = ScalarParameter() {parameterName=LerpRibbon}
  n6 = ScalarParameter() {defaultValue=1, parameterName=MasterSpeed}
  n7 = Multiply(A=ScalarParameter() {defaultValue=1, parameterName=DistortionSpeed})
  n8 = Panner(
   Coordinate=Multiply(A=AppendVector(A=@n1, B=@n2), B=TextureCoordinate())
   Speed=Multiply(A=LinearInterpolate(A=@n3, B=@n4, Alpha=@n5), B=Multiply(A=@n6, B=@n7))
  ) {speedY=-0.4}
  n9 = TextureCoordinate() {uTiling=0.7, vTiling=1.5}
  n10 = Constant2Vector() {g=0.75, r=0.9}
  n11 = Constant() {r=1}
  n12 = Constant()
  n13 = ComponentMask(None=TextureCoordinate()) {r=true}
  n14 = Multiply(
   A=Subtract(A=SphereMask(A=@n9, B=@n10, Radius=@n11, Hardness=@n12), B=Multiply(A=@n13) {constB=0.7})
   B=ScalarParameter() {parameterName=SphereMAskAddDist}
  )
  Add(A=@n8, B=@n14)
  Constant()
  n15 = TextureCoordinate()
  n16 = ScalarParameter() {parameterName=LerpRibbon}
  n17 = LinearInterpolate(A=ComponentMask(None=@n15) {g=true}, B=OneMinus(None=ComponentMask(None=@n15) {r=true}), Alpha=@n16)
  n18 = DynamicParameter() {parameterIndex=1, paramNames [Offset1, Offset2, DistortionMult, AudioReactAdd]}
  n19 = TextureSample(UVs=Add(A=@n8, B=AppendVector(A=@n18.Offset1, B=@n18.Offset1))) {samplerType=SAMPLERTYPE_LinearGrayscale, texture=&T_SimplePerlin_Greyscale}
  n20 = ScalarParameter() {parameterName=OpacityDistortMult}
  n21 = Power(Base=Add(A=@n17, B=Multiply(A=@n19.R, B=@n20)), Exp=ScalarParameter() {defaultValue=2.3, parameterName=OpacityMaskPWR}) {constExponent=2.3}
  n22 = Sine(None=@n17) {period=2}
  n23 = ScalarParameter() {defaultValue=1.0445, parameterName=LinesGradPWR}
  n24 = Add(A=Add(A=Add(A=@n15, B=@n14), B=OneMinus(None=@n18.AudioReactAdd)), B=Multiply(A=@n19.R, B=@n18.DistortionMult) {constB=0.35})
  n25 = ComponentMask(None=@n24) {g=true}
  n26 = OneMinus(None=ComponentMask(None=@n24) {r=true})
  n27 = DynamicParameter() {parameterIndex=2, defaultValue=<rgba 1 0.6 1 1>, paramNames [PanSpeedDive, HandGradOffset, Param3, Param4]}
  n28 = Multiply(A=@n6, B=ScalarParameter() {defaultValue=-0.45, parameterName=LinesSpeed})
  n29 = Multiply(
   A=Add(A=LinearInterpolate(A=@n25, B=@n26, Alpha=@n16), B=Multiply(A=@n27.PanSpeedDive, B=@n28) {constB=-0.35})
   B=ScalarParameter() {defaultValue=3.4, parameterName=LinesCount}
  ) {constB=6}
  n30 = Saturate(None=Power(Base=Sine(None=@n29), Exp=ScalarParameter() {defaultValue=0.5012, parameterName=Lines_PWR}) {constExponent=18.9182})
  n31 = Add(
   A=MaterialFunctionCall() {materialFunction=&BoundingBoxBased_0-1_UVW}.B
   B=@n27.HandGradOffset
  ) {constB=0.6}
  n32 = Saturate(None=Power(Base=OneMinus(None=@n31)) {constExponent=0.6})
  n33 = Multiply(
   A=Saturate(None=Power(Base=@n22, Exp=@n23) {constExponent=0.3})
   B=LinearInterpolate(A=@n30, B=Multiply(A=@n30, B=@n32), Alpha=ScalarParameter() {parameterName=LerpHand})
  )
  n34 = ScalarParameter() {parameterName=OpacityLinesAdd_Mult}
  n35 = Saturate(None=Add(A=OneMinus(None=@n21), B=Multiply(A=@n33, B=@n34)))
  n36 = OneMinus(None=Fresnel() {exponent=2})
  n37 = LinearInterpolate(A=@n35, B=Multiply(A=@n35, B=Saturate(None=@n36)), Alpha=ScalarParameter() {parameterName=UseFresnelFade})
  n38 = Sine(None=ComponentMask(None=TextureCoordinate()) {g=true}) {period=2}
  n39 = Multiply(A=LinearInterpolate(A=@n37, B=Multiply(A=@n37, B=@n38), Alpha=ScalarParameter() {parameterName=LerpRibbon}), B=ParticleColor().A)
  n40 = LinearInterpolate(
   A=@n39
   B=ShadingPathSwitch(Default=DepthFade(Opacity=@n39), Forward=@n39, Mobile=@n39)
   Alpha=ScalarParameter() {parameterName=LerpDepthFade}
  )
  LinearInterpolate(A=@n40, B=LinearInterpolate(A=Multiply(A=@n40) {constB=0.1}, B=@n40, Alpha=@n32), Alpha=ScalarParameter() {parameterName=LerpHand})
  n41 = VectorParameter() {parameterName=Color1, defaultValue=<rgba 0.1042 0.6618 4 1>}
  n42 = Multiply(A=@n32)
  n43 = Multiply(
   A=VectorParameter() {parameterName=Color4, defaultValue=<rgba 4 0 3.9281 1>}.RGB
   B=LinearInterpolate(B=@n42, Alpha=ScalarParameter() {parameterName=LerpHand}) {constA=1}
  )
  n44 = Add(A=TextureCoordinate(), B=Multiply(A=@n19.R) {constB=0.15})
  n45 = ComponentMask(None=@n44) {g=true}
  n46 = OneMinus(None=ComponentMask(None=@n44) {r=true})
  n47 = ScalarParameter() {parameterName=LerpRibbon}
  n48 = Saturate(None=Power(Base=LinearInterpolate(A=@n45, B=@n46, Alpha=@n47), Exp=ScalarParameter() {defaultValue=2.7783, parameterName=ColorGradPWR}) {constExponent=2.3})
  MaterialFunctionCall(Emissive=MaterialFunctionCall(
    A=LinearInterpolate(A=@n41.RGB, B=@n43, Alpha=@n48)
    B=VectorParameter() {parameterName=Color2, defaultValue=<rgba 1.2486 0.1639 1.6562 1>}.RGB
    C=VectorParameter() {parameterName=Color3, defaultValue=<rgba 0.0948 0.3884 1.4 1>}.RGB
    Alpha=@n33
   ) {materialFunction=&Lerp_3Color}.Result) {materialFunction=&MF_FX_EmissiveAdjustments}
  n49 = TextureSample(UVs=Panner() {speedY=-0.3}) {texture=/Game/Effects/Fort_Effects/Emotes/UltraLight/Texture/T_SoftNoise_BC7}
  n50 = VertexNormalWS()
  n51 = Constant3Vector() {constant=<rgba 0 0 1 0>}
  n52 = Multiply(A=ComponentMask(None=TextureCoordinate()) {g=true}, B=Multiply(A=@n49.R, B=Multiply(A=@n50, B=@n51)))
  n53 = Constant3Vector() {constant=<rgba 1 1 0 0>}
  n54 = Multiply(
   A=Multiply(
    A=Multiply(A=@n50, B=@n53)
    B=MaterialFunctionCall() {materialFunction=&BoundingBoxBased_0-1_UVW}.B
   )
   B=@n49.R
  )
  Multiply(
   A=LinearInterpolate(A=Multiply(A=@n52) {constB=22}, B=Multiply(A=@n54) {constB=11}, Alpha=ScalarParameter() {parameterName=Lerp_WPOHand})
   B=ScalarParameter() {defaultValue=1, parameterName=WPOMult}
  )
  n55 = CameraVectorWS()
  n56 = VertexNormalWS()
  Multiply(A=Saturate(None=DotProduct(A=@n55, B=@n56)))
  Saturate(None=Add(A=@n42, B=@n48))
  Time()
 ]
}
