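# Translucent, unlit, two-sided Niagara mesh particle material for the PolarityWinn skydiving contrail.
# Perlin-distorted sine waves form line displacement, with added ring wobble and normal-based expansion tapered along V.
# Two dynamic parameters drive displacement strengths, wave offsets and dive pan speed, so the particle system controls the motion rather than Time.
# Sine masks along V shape opacity and blend three colors plus a glow band, tinted by particle color through MF_FX_EmissiveAdjustments.
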Material /BRCosmetics/Athena/Effects/Particles/Skydiving/NS_Contrail_PolarityWinnField/M_Contrail_PolarityWinn_Spline
graph {
 properties {
  bAutomaticallySetUsageInEditor=false
  bUsedWithNiagaraMeshParticles=true
  blendMode=BLEND_Translucent
  pixelDepthOffsetMode=PDOM_Legacy
  shadingModel=MSM_Unlit
  twoSided=true
  usedShadingModels=Unlit
 }
 nodes [
  n1 = Constant3Vector() {constant=<rgba 0 0 1 0>}
  n2 = DynamicParameter() {parameterIndex=1, defaultValue=<rgba 1 1 1 0>, paramNames [Offset1,Offset2,PanSpeedDive,LerpGlow]}
  n3 = Multiply(A=Add(A=TextureCoordinate(), B=AppendVector(A=@n2.Offset2, B=@n2.Offset2)), B=Constant2Vector() {g=0.3, r=1})
  n4 = AppendVector(A=Constant(), B=Multiply(A=@n2.PanSpeedDive) {constB=0.1})
  n5 = Multiply(A=TextureSample(UVs=Add(A=@n3, B=@n4)) {samplerType=SAMPLERTYPE_LinearGrayscale, texture=&T_SimplePerlin_Greyscale}.R) {constB=0.16}
  n6 = TextureCoordinate()
  n7 = Add(A=ComponentMask(None=Add(A=@n5, B=@n6)) {r=true}) {constB=0}
  n8 = Multiply(A=ScalarParameter() {defaultValue=1, parameterName=MasterSpeed}, B=@n2.PanSpeedDive)
  n9 = ScalarParameter() {defaultValue=0.7, parameterName=LinesSpeedMult}
  n10 = Multiply(A=@n8, B=Multiply(A=Constant() {r=0.12}, B=@n9)) {constB=0.12}
  n11 = Multiply(A=Add(A=@n7, B=Add(A=@n10, B=@n2.Offset1)), B=Constant() {r=6}) {constB=6}
  n12 = Divide(A=Add(A=Sine(None=@n11)))
  n13 = Constant() {r=-0.11}
  n14 = Add(A=Multiply(A=@n8, B=Multiply(A=@n13, B=@n9)) {constB=-0.11}, B=@n2.Offset2)
  n15 = Sine(None=Multiply(A=Add(A=@n7, B=@n14), B=Constant() {r=5}) {constB=5})
  n16 = Multiply(A=@n12, B=Divide(A=Add(A=@n15)))
  n17 = ScalarParameter() {defaultValue=1.4, parameterName=LinesPWR}
  n18 = DynamicParameter() {defaultValue=<rgba 80 0.02 7 7>, paramNames [WPO_Strength, WPO_Bottom_Mult, Expand_Top, Ring_Wobble]}
  n19 = Multiply(A=Saturate(None=Power(Base=@n16, Exp=@n17) {constExponent=1.4}), B=@n18.WPO_Strength) {constB=150}
  n20 = ComponentMask(None=TextureCoordinate()) {g=true}
  n21 = Saturate(None=Round(None=@n20))
  n22 = LinearInterpolate(
   A=Multiply(A=@n19, B=ScalarParameter() {defaultValue=1, parameterName=WPO_Top_Mult})
   B=Multiply(A=@n19, B=@n18.WPO_Bottom_Mult) {constB=0.02}
   Alpha=@n21
  ) {constB=0}
  n23 = Multiply(A=@n12, B=Constant3Vector() {constant=<rgba 1 1 0 0>})
  n24 = OneMinus(None=@n20)
  n25 = VertexNormalWS()
  Add(
   A=Add(A=Multiply(A=@n1, B=@n22), B=Multiply(A=@n23, B=@n18.Ring_Wobble) {constB=0}) {constB=0}
   B=Multiply(A=Multiply(A=@n24, B=@n25), B=@n18.Expand_Top) {constB=0}
  )
  n26 = ParticleColor()
  n27 = OneMinus(None=Sine(None=@n20) {period=2})
  n28 = Saturate(None=Power(Base=@n27, Exp=ScalarParameter() {defaultValue=2.2, parameterName=OpacityPWR}) {constExponent=2.2})
  n29 = OneMinus(None=Sine(None=@n20) {period=2})
  n30 = ScalarParameter() {defaultValue=1.8, parameterName=GlowMaskPWR}
  n31 = Saturate(None=Round(None=Power(Base=@n29, Exp=@n30) {constExponent=1.8}))
  n32 = Multiply(A=Saturate(None=Sine(None=@n20) {period=2}), B=@n31)
  n33 = Multiply(A=@n26.A, B=LinearInterpolate(A=LinearInterpolate(Alpha=@n28) {constA=1}, B=Multiply(A=@n32), Alpha=@n31))
  n34 = Saturate(None=@n20)
  LinearInterpolate(A=@n33, B=Multiply(A=@n33), Alpha=Saturate(None=Power(Base=@n34) {constExponent=3}))
  n35 = Power(Base=Sine(None=@n20) {period=2}) {constExponent=1.3}
  n36 = VectorParameter() {parameterName=Color2, defaultValue=<rgba 2.3177 4.4052 5 1>}
  n37 = ScalarParameter() {defaultValue=6, parameterName=ColorPWR}
  n38 = Subtract(A=Saturate(None=Power(Base=@n27, Exp=@n37) {constExponent=6}), B=@n31)
  n39 = LinearInterpolate(
   A=LinearInterpolate(
    A=VectorParameter() {parameterName=Color1, defaultValue=<rgba 1.3099 0.0599 2 1>}.RGB
    B=VectorParameter() {parameterName=Color3, defaultValue=<rgba 0.2404 0.0044 0.4948 1>}.RGB
    Alpha=Saturate(None=@n35)
   )
   B=LinearInterpolate(A=@n36.RGB, B=Multiply(A=@n36.RGB) {constB=0.3}, Alpha=@n21)
   Alpha=Saturate(None=Multiply(A=@n38) {constB=15})
  )
  n40 = VectorParameter() {parameterName=GlowColor, defaultValue=<rgba 0 1.7057 14 3>}
  MaterialFunctionCall(Emissive=Multiply(A=@n26.RGB, B=LinearInterpolate(A=@n39, B=@n40.RGB, Alpha=@n31))) {materialFunction=&MF_FX_EmissiveAdjustments}
  Multiply(A=@n25, B=Constant3Vector() {constant=<rgba 0 0 1 0>})
  Multiply(A=@n2.PanSpeedDive, B=ScalarParameter() {defaultValue=1, parameterName=MasterSpeed})
  Panner(Coordinate=@n3) {bFractionalPart=true, speedY=0.1}
  Round(None=Add(A=@n20) {constB=-0.28})
  n41 = Sine(None=@n20) {period=2}
  n42 = OneMinus(None=@n41)
  Saturate(None=Multiply(A=@n41, B=Power(Base=@n42) {constExponent=3}))
  TextureObjectParameter() {
   group=NoiseMask
   parameterName=NoiseMask_Texture
   sortPriority=0
   texture=/Game/Effects/Fort_Effects/Emotes/UltraLight/Texture/T_SoftNoise_BC7
  }
  TextureObjectParameter() {
   group=Hue
   parameterName=Hue_Texture
   samplerType=SAMPLERTYPE_LinearGrayscale
   sortPriority=0
   texture=&T_SimplePerlin_Greyscale
  }
  Time()
  Time()
 ]
}
