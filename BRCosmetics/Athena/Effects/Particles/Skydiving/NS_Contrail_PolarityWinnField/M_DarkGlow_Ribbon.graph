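# Translucent unlit Niagara ribbon material that draws a soft, dark-toned glow for the PolarityWinnField skydiving contrail.
# Two color parameters, purple and blue by default, blend along U through a power curve and pass through MF_FX_EmissiveAdjustments.
# A texture-free mask from a sine of V, tightened by a power, fades both ribbon edges and is scaled by particle alpha.
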Material /BRCosmetics/Athena/Effects/Particles/Skydiving/NS_Contrail_PolarityWinnField/M_DarkGlow_Ribbon
graph {
 properties {
  bAutomaticallySetUsageInEditor=false
  bUsedWithNiagaraRibbons=true
  blendMode=BLEND_Translucent
  shadingModel=MSM_Unlit
  usedShadingModels=Unlit
 }
 nodes [
  n1 = TextureCoordinate()
  n2 = Power(Base=ComponentMask(None=@n1) {r=true})
  MaterialFunctionCall(Emissive=LinearInterpolate(
    A=VectorParameter() {parameterName=Color_1, defaultValue=<rgba 0.1317 0.0133 0.2708 1>}.RGB
    B=VectorParameter() {parameterName=Color_2, defaultValue=<rgba 0.0269 0.175 0.3854 1>}.RGB
    Alpha=Saturate(None=@n2)
   )) {materialFunction=&MF_FX_EmissiveAdjustments}
  n3 = Sine(None=ComponentMask(None=@n1) {g=true}) {period=2}
  Multiply(A=ParticleColor().A, B=Saturate(None=Power(Base=@n3) {constExponent=1.1}))
 ]
}
