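# Niagara skydiving contrail cosmetic (PolarityWinnField) pairing glowing hand effects and hand ribbons with a ground-style circle and three ring-wave meshes.
# Hand meshes and glow sprites follow skeletal-mesh sockets; ribbons are pendulum-constrained particle chains trailing against owner velocity, each drawn by two renderers.
# Contrail setup dive value, glide lean and a fake audio-reactivity loop drive scale, width, alpha and material parameters.
# Time-of-day and front-end modifiers retune brightness and offsets.
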NiagaraSystem /BRCosmetics/Athena/Effects/Particles/Skydiving/NS_Contrail_PolarityWinnField/NS_Contrail_PolarityWinnField
dependencies {
 usedDataInterfaces [
  &NiagaraDataInterfaceCurve
  &NiagaraDataInterfaceArrayFloat3
  &NiagaraDataInterfaceArrayInt32
  &NiagaraDataInterfaceArrayFloat
  &NiagaraDataInterfaceSkeletalMesh
  &NiagaraDataInterfaceColorCurve
 ]
 usedDynamicInputs [
  &SimulationPosition
  &MakeIntFromFloat
  &Waveform
  /Niagara/DynamicInputs/Lerp/SmoothLerpOverTime_Float
  &SetBoolByFloatComparison
  &RemapRange
  &MakeFloatFromLinearColor
 ]
 usedModules [
  &self:SystemSpawnScript.SystemScriptSource.SystemScriptGraph.NiagaraNodeAssignment_0.SetVariables_#1
  &self:SystemSpawnScript.SystemScriptSource.SystemScriptGraph.NiagaraNodeAssignment_1.SetVariables_#2
  &self:SystemSpawnScript.SystemScriptSource.SystemScriptGraph.NiagaraNodeAssignment_5.SetVariables_#3
  &self:SystemSpawnScript.SystemScriptSource.SystemScriptGraph.NiagaraNodeAssignment_8.SetVariables_#4
  &SystemState
  /Game/Effects/Niagara/Modules/Contrails/FN_Contrail_Setup
  &FN_Integrator
  /Game/Effects/Niagara/Modules/Audio/FN_FakeAudioReactivityLoop
  &self:SystemSpawnScript.SystemScriptSource.SystemScriptGraph.NiagaraNodeAssignment_2.SetVariables_#5
  &self:SystemSpawnScript.SystemScriptSource.SystemScriptGraph.NiagaraNodeAssignment_3.SetVariables_#6
  &self:SystemSpawnScript.SystemScriptSource.SystemScriptGraph.NiagaraNodeAssignment_6.SetVariables_#7
  &self:SystemSpawnScript.SystemScriptSource.SystemScriptGraph.NiagaraNodeAssignment_9.SetVariables_#8
  &self:SystemSpawnScript.SystemScriptSource.SystemScriptGraph.NiagaraNodeAssignment_10.SetVariables_#9
  &self:SystemSpawnScript.SystemScriptSource.SystemScriptGraph.NiagaraNodeAssignment_11.SetVariables_#10
 ]
}
emitters {
 BotCricle {
  emitterUpdateScript [
   &EmitterState()
   &SpawnBurst_Instantaneous(Spawn Count=1, Spawn Probability=HlslExpression() {hlslExpression="Emitter.LoopCount == 0"}, Use Spawn Probability=true)
  ]
  particleSpawnScript [
   &InitializeParticle(
    Lifetime Max=1
    Lifetime Mode=::NewEnumerator1
    Mesh Scale Mode=::NewEnumerator3
    Position Offset=&FN_FrontEndVector3Modifier(Gameplay=<0 0 0>, Front End=<0 0 20>)
    UsePositionOffset=true
    Write Material Random=true
   )
   &InitialMeshOrientation(Mesh Axis To Orient=<0 1 0>, Mesh Orientation Mode=::NewEnumerator2, Orientation Vector=<0 0 1>, Use Rotation Vector=false)
   /SetVariables() {isSetParameters=true}
  ]
  particleUpdateScript [
   &ParticleState(Kill Particles When Lifetime Has Elapsed=false, Let Infinitely Lived Particles Die When Emitter Deactivates=true)
   /SetVariables(Particles.CustomSpeed:NiagaraFloat=&Add_Float(
     A=$Particles.CustomSpeed
     B=&Multiply_Float(
      A=&Add_Float(A=1, B=&Multiply_Float(A=0.77, B=$User.ParaGlideLeanAlpha))
      B=$Engine.DeltaTime
     )
    )) {isSetParameters=true}
   &DynamicMaterialParameters(
    Index 0 Param 1=&MergeHelper_FloatSwitch(0=50, 1=45, 2=40, 3=?, Index=&SpawnGroup(), Num=::NewEnumerator1)
    Index 0 Param 2=0.08
    Index 0 Param 3=&MergeHelper_FloatSwitch(0=8, 1=1, 2=-8, 3=?, Index=&SpawnGroup(), Num=::NewEnumerator1)
    Index 0 Param 4=3
    Index 1 Param 1=&MergeHelper_FloatSwitch(0=0, 1=0.4, 2=0.7, 3=?, Index=&SpawnGroup(), Num=::NewEnumerator1)
    Index 1 Param 2=&MergeHelper_FloatSwitch(0=0.6, 1=0.8, 2=0.2, 3=?, Index=&SpawnGroup(), Num=::NewEnumerator1)
    Index 1 Param 3=&Add_Float(
     A=&Lerp_Float(A=0.07, B=0.15, Alpha=$System.AudioReactivity)
     B=&Lerp_Float(A=0, B=0.05, Alpha=$System.AudioReactivity2)
    )
    Index 1 Param 4=&Add_Float(A=&Multiply_Float(A=$System.AudioReactivity, B=0), B=0)
    Index 2 Param 1=$Particles.CustomSpeed
    Module.Index0Float0Write=true
    Module.Index0Float1Write=true
    Module.Index0Float2Write=true
    Module.Index0Float3Write=true
    Module.Index1Float0Write=true
    Module.Index1Float1Write=true
    Module.Index1Float2Write=true
    Module.Index1Float3Write=true
    Module.Index2Float0Write=true
    Module.Index2Float1Write=true
    Module.Index2Float2Write=true
    Module.Index2Float3Write=true
    Write Parameter Index 1=::NewEnumerator1
    Write Parameter Index 2=::NewEnumerator1
   )
   &ScaleColor(
    Scale Alpha=&Multiply_Float(
     A=&FNTimeOfDayScalar(Night Scale=0.7, Morning Scale=0.9, Evening Scale=0.8)
     B=&Multiply_Float(
      A=&FN_FrontEndFloatModifier(Gameplay=1.3, Front End=0.5)
      B=&Add_Float(
       A=&Multiply_Float(
        A=&Add_Float(
         A=&Lerp_Float(A=0.06, B=0.08, Alpha=$System.AudioReactivity)
         B=&Lerp_Float(A=0, B=0.05, Alpha=$System.AudioReactivity2)
        )
        B=&FloatFromCurve(FloatCurve=DataInterface() {propertyValues {Curve {keys [<at 0 0 RCIM_Cubic>,<at 0.4 1>]}}}, CurveIndex=$Engine.System.Age)
       )
       B=&Lerp_Float(A=0, B=0.05, Alpha=$System.FN_Contrail_Setup.Dive Value)
      )
     )
    )
    ScaleA=true
    ScaleRGB=true
   )
  ]
  rendererClasses [&NiagaraMeshRendererProperties]
  renderers [{rendererClass=&NiagaraMeshRendererProperties, rendererIndex=0}]
 }
 Hand_Glow {
  emitterUpdateScript [
   &EmitterState(Life Cycle Mode=::NewEnumerator1, Loop Behavior=::NewEnumerator1, Loop Duration Mode=::NewEnumerator1, UseLoopDelay=false)
   &SpawnBurst_Instantaneous(Spawn Count=2, Use Spawn Probability=false)
  ]
  particleSpawnScript [
   &FN_SetSpawnGroup(Number of Spawn Groups=2)
   &InitializeParticle(
    Color=<rgba 0.0435 0.1013 1.5 2>
    Lifetime=2
    Mass Mode=::NewEnumerator1
    Position Offset=<0 0 100>
    Sprite Size Mode=::NewEnumerator3
    Uniform Sprite Size=120
    UsePositionOffset=false
    Write Material Random=true
   ) {module=InitializeParticle001}
  ]
  particleUpdateScript [
   &ParticleState(Kill Particles When Lifetime Has Elapsed=false)
   &SkeletalMeshLocation(
    Bone / Socket Index=&SpawnGroup()
    Bone Sampling Mode=::NewEnumerator4
    MaskSpawnGroup=false
    Sampled Position Offset=&MergeHelper_VectorSwitch(0=<6 0 0>, 1=<-6 0 0>, 2=?, 3=?, Index=&SpawnGroup(), Num=::NewEnumerator0)
    Skeletal Mesh=$System.Skeletal Mesh
   ) {module=SkeletalMeshLocation001}
   &ScaleSpriteSize(
    Uniform Curve Scale=&Sine(Normalized Angle=&Add_Float(A=$Emitter.Age, B=0), Period=0.8, Scale=0.03, Bias=0.65)
    Uniform Curve Sprite Scale=DataInterface() {propertyValues {Curve {keys [{tangentMode=RCTM_User, time=0, value=1}, {tangentMode=RCTM_User, time=1, value=1}]}}}
   )
   &CameraOffset()
   /SetVariables(Particles.Color_Decal:LinearColor=&MakeLinearColor(
     R:NiagaraFloat=0
     G:NiagaraFloat=0.124
     B:NiagaraFloat=1
     A:NiagaraFloat=&FloatFromCurve(
      FloatCurve=DataInterface() {propertyValues {Curve {keys [<at 0.1 0.3382 RCIM_Cubic>,<at 0.7 0.85>]}}}
      CurveIndex=$System.Age
      Scale Curve=0.8
     )
    )) {isSetParameters=true}
   &ScaleColor(
    Scale Alpha=&FloatFromCurve(
     FloatCurve=DataInterface() {propertyValues {Curve {keys [<at 0 0.7 RCIM_Cubic>,<at 0.5 1 RCIM_Cubic>,<at 1 0.7 RCIM_Cubic>]}}}
     CurveIndex=Linked() {linkedVariable {name=NPC.FortniteNPC.TimeOfDay_0-1, type=NiagaraFloat}}
     Scale Curve=&FN_FrontEndFloatModifier(Gameplay=1, Front End=0.7)
    )
    ScaleA=true
    ScaleRGB=true
   )
  ]
  rendererClasses [&NiagaraSpriteRendererProperties]
  renderers [
   {rendererClass=&NiagaraSpriteRendererProperties}
   {rendererClass=&NiagaraSpriteRendererProperties, rendererIndex=1}
  ]
 }
 Hands {
  emitterUpdateScript [
   &EmitterState()
   &SpawnBurst_Instantaneous(Spawn Count=2, Spawn Probability=HlslExpression() {hlslExpression="Emitter.LoopCount == 0"}, Use Spawn Probability=true)
  ]
  particleSpawnScript [
   &FN_SetSpawnGroup(Number of Spawn Groups=2)
   &InitializeParticle(
    Lifetime Max=1
    Lifetime Mode=::NewEnumerator1
    Mesh Scale=<0.18 0.18 0.18>
    Mesh Scale Mode=::NewEnumerator4
    UsePositionOffset=false
    Write Material Random=true
   )
   &InitialMeshOrientation(
    Mesh Orientation Mode=::NewEnumerator2
    Orientation Vector=&MergeHelper_VectorSwitch(0=<0 1 0>, 1=<0 1 0>, 2=?, 3=?, Index=&SpawnGroup(), Num=::NewEnumerator0)
    Rotation=&MergeHelper_VectorSwitch(0=<0 0.06 0>, 1=<0 -0.06 0>, 2=?, 3=?, Index=&SpawnGroup(), Num=::NewEnumerator0)
    Use Rotation Vector=true
   ) {module=InitialMeshOrientation001}
   /SetVariables() {isSetParameters=true}
  ]
  particleUpdateScript [
   &ParticleState(Kill Particles When Lifetime Has Elapsed=false, Let Infinitely Lived Particles Die When Emitter Deactivates=true)
   /SetVariables(Particles.CustomSpeed:NiagaraFloat=&Add_Float(
     A=$Particles.CustomSpeed
     B=&Multiply_Float(
      A=&Add_Float(A=1, B=&Multiply_Float(A=1, B=$User.ParaGlideLeanAlpha))
      B=$Engine.DeltaTime
     )
    )) {isSetParameters=true}
   &UpdateMeshOrientation(
    Facing Interpolation=false
    Orientation Method=::NewEnumerator1
    Override Quat=false
    TransformFacing=false
    TransformUp=false
    Up Direction:Vector3f=&Lerp_Vector(
     A=&FN_FrontEndVector3Modifier(
      Gameplay=&MergeHelper_VectorSwitch(0=<0 -0.4 1>, 1=<0 0.4 1>, 2=?, 3=?, Index=&SpawnGroup(), Num=::NewEnumerator0)
      Front End=&MergeHelper_VectorSwitch(0=<0 0 1>, 1=<0 0 1>, 2=?, 3=?, Index=&SpawnGroup(), Num=::NewEnumerator0)
     )
     B=&FN_FrontEndVector3Modifier(
      Gameplay=&MergeHelper_VectorSwitch(0=<0 2 1>, 1=<0 -2 1>, 2=?, 3=?, Index=&SpawnGroup(), Num=::NewEnumerator0)
      Front End=&MergeHelper_VectorSwitch(0=<0 -0.4 1>, 1=<0 0.4 1>, 2=?, 3=?, Index=&SpawnGroup(), Num=::NewEnumerator0)
     )
     Alpha=$User.ParaGlideLeanAlpha
    )
   ) {module=UpdateMeshOrientation002}
   &ScaleMeshSize(Scale Factor=&Lerp_Vector(A=<1.15 1.15 1.15>, B=<1.4 1.4 1.5>, Alpha=$System.AudioReactivity))
   &DynamicMaterialParameters(
    Index 0 Param 1=&MergeHelper_FloatSwitch(0=300, 1=967.8755, 2=40, 3=?, Index=&SpawnGroup(), Num=::NewEnumerator1)
    Index 0 Param 2=0.176
    Index 1 Param 1=&MergeHelper_FloatSwitch(0=0, 1=0.4, 2=0.7, 3=?, Index=&SpawnGroup(), Num=::NewEnumerator1)
    Index 1 Param 2=&MergeHelper_FloatSwitch(0=0.6, 1=0.9, 2=0.2, 3=?, Index=&SpawnGroup(), Num=::NewEnumerator1)
    Index 1 Param 3=&Add_Float(
     A=&Lerp_Float(A=0.3, B=0.35, Alpha=$System.AudioReactivity)
     B=&Lerp_Float(A=0, B=0.1, Alpha=$System.AudioReactivity2)
    )
    Index 1 Param 4=&Multiply_Float(A=$System.AudioReactivity, B=0.15)
    Index 2 Param 1=$Particles.CustomSpeed
    Index 2 Param 2=&FN_FrontEndFloatModifier(Gameplay=0.58, Front End=0.55)
    Module.Index0Float0Write=true
    Module.Index0Float1Write=true
    Module.Index0Float2Write=true
    Module.Index0Float3Write=true
    Module.Index1Float0Write=true
    Module.Index1Float1Write=true
    Module.Index1Float2Write=true
    Module.Index1Float3Write=true
    Module.Index2Float0Write=true
    Module.Index2Float1Write=true
    Module.Index2Float2Write=true
    Module.Index2Float3Write=true
    Write Parameter Index 1=::NewEnumerator1
    Write Parameter Index 2=::NewEnumerator1
   )
   &ScaleColor(
    Scale Alpha=&Multiply_Float(
     A=&FNTimeOfDayScalar(Night Scale=0.65, Morning Scale=0.85, Evening Scale=0.75)
     B=&Multiply_Float(
      A=&FN_FrontEndFloatModifier(Gameplay=4.5, Front End=2.7)
      B=&Add_Float(
       A=&Multiply_Float(
        A=&Add_Float(
         A=&Lerp_Float(A=0.06, B=0.1, Alpha=$System.AudioReactivity)
         B=&Lerp_Float(A=0, B=0.05, Alpha=$System.AudioReactivity2)
        )
        B=&FloatFromCurve(
         FloatCurve=DataInterface() {propertyValues {Curve {keys [<at 0.1 0 RCIM_Cubic>,<at 0.5 1>]}}}
         CurveIndex=$Engine.System.Age
         Scale Curve=1.7
        )
       )
       B=&Lerp_Float(A=0, B=0.05, Alpha=$System.FN_Contrail_Setup.Dive Value)
      )
     )
    )
    ScaleA=true
    ScaleRGB=true
   )
   &SkeletalMeshLocation(
    Bone / Socket Index=&SpawnGroup()
    Bone Sampling Mode=::NewEnumerator4
    MaskSpawnGroup=false
    Orientation Sampling=::NewEnumerator1
    Sampled Position Offset=&FN_FrontEndVector3Modifier(
     Gameplay=&FN_FrontEndVector3Modifier(
      Gameplay=&MergeHelper_VectorSwitch(0=<8 1 -0.5>, 1=<-8 -1 0.5>, 2=?, 3=?, Index=&SpawnGroup(), Num=::NewEnumerator0)
      Front End=&MergeHelper_VectorSwitch(0=<8 1 0>, 1=<-8 -1 0>, 2=?, 3=?, Index=&SpawnGroup(), Num=::NewEnumerator0)
     )
     Front End=&Lerp_Vector(
      A=&MergeHelper_VectorSwitch(0=<8 1 0>, 1=<-8 -1 0>, 2=?, 3=?, Index=&SpawnGroup(), Num=::NewEnumerator0)
      B=&MergeHelper_VectorSwitch(0=<8 -3 0>, 1=<-8 3 0>, 2=?, 3=?, Index=&SpawnGroup(), Num=::NewEnumerator0)
      Alpha=$User.ParaGlideLeanAlpha
     )
    )
    Skeletal Mesh=$System.Skeletal Mesh
   )
  ]
  rendererClasses [&NiagaraMeshRendererProperties]
  renderers [{rendererClass=&NiagaraMeshRendererProperties, rendererIndex=0}]
 }
 RingWaves {
  emitterUpdateScript [
   &EmitterState()
   &SpawnBurst_Instantaneous(Spawn Count=3, Spawn Probability=HlslExpression() {hlslExpression="Emitter.LoopCount == 0"}, Use Spawn Probability=true)
  ]
  particleSpawnScript [
   &FN_SetSpawnGroup(Number of Spawn Groups=3)
   &InitializeParticle(
    Lifetime Max=1
    Lifetime Mode=::NewEnumerator1
    Mesh Scale=&MergeHelper_VectorSwitch(0=<0.37 0.37 0.25>, 1=<0.37 0.37 0.23>, 2=<0.37 0.37 0.2>, 3=?, Index=&SpawnGroup(), Num=::NewEnumerator1)
    Mesh Scale Mode=::NewEnumerator4
    Position Offset=&FN_FrontEndVector3Modifier(Gameplay=<0 0 0>, Front End=<0 0 20>)
    UsePositionOffset=true
    Write Material Random=true
   )
   &InitialMeshOrientation(
    Mesh Orientation Mode=::NewEnumerator2
    Rotation=&MergeHelper_VectorSwitch(0=<0 0 0>, 1=<0 0 0.3>, 2=<0 0 0.8>, 3=?, Index=&SpawnGroup(), Num=::NewEnumerator1)
    Use Rotation Vector=true
   )
   &CameraOffset(Camera Offset Amount=&MergeHelper_FloatSwitch(0=0, 1=0, 2=0, 3=-3, Index=&SpawnGroup(), Num=::NewEnumerator2))
   /SetVariables() {isSetParameters=true}
  ]
  particleUpdateScript [
   &ParticleState(Kill Particles When Lifetime Has Elapsed=false, Let Infinitely Lived Particles Die When Emitter Deactivates=true)
   &ScaleMeshSize(Scale Factor=&Lerp_Vector(A=<1 1 1>, B=<1 1 1.3>, Alpha=$System.FN_Contrail_Setup.Dive Value))
   /SetVariables(Particles.CustomSpeed:NiagaraFloat=&Add_Float(
     A=$Particles.CustomSpeed
     B=&Multiply_Float(
      A=&Add_Float(A=1, B=&Multiply_Float(A=0.5, B=$User.ParaGlideLeanAlpha))
      B=$Engine.DeltaTime
     )
    )) {isSetParameters=true}
   &DynamicMaterialParameters(
    Index 0 Param 1=&Add_Float(
     A=&Lerp_Float(
      A=&MergeHelper_FloatSwitch(0=19, 1=16, 2=13, 3=?, Index=&SpawnGroup(), Num=::NewEnumerator1)
      B=&Multiply_Float(
       A=&MergeHelper_FloatSwitch(0=50, 1=45, 2=40, 3=?, Index=&SpawnGroup(), Num=::NewEnumerator1)
       B=1.3
      )
      Alpha=$System.AudioReactivity
     )
     B=&Lerp_Float(
      A=0
      B=&Multiply_Float(
       A=&MergeHelper_FloatSwitch(0=20, 1=18, 2=16, 3=?, Index=&SpawnGroup(), Num=::NewEnumerator1)
       B=1.3
      )
      Alpha=$System.AudioReactivity2
     )
    )
    Index 0 Param 2=0.08
    Index 0 Param 3=&MergeHelper_FloatSwitch(0=11, 1=1, 2=-11, 3=8, Index=&SpawnGroup(), Num=::NewEnumerator2)
    Index 0 Param 4=4
    Index 1 Param 1=&MergeHelper_FloatSwitch(0=0, 1=0.4, 2=0.7, 3=0, Index=&SpawnGroup(), Num=::NewEnumerator2)
    Index 1 Param 2=&MergeHelper_FloatSwitch(0=0.6, 1=0.8, 2=0.2, 3=0.6, Index=&SpawnGroup(), Num=::NewEnumerator2)
    Index 1 Param 3=$Particles.CustomSpeed
    Index 1 Param 4=&MergeHelper_FloatSwitch(0=0, 1=0, 2=0, 3=1, Index=&SpawnGroup(), Num=::NewEnumerator2)
    Module.Index0Float0Write=true
    Module.Index0Float1Write=true
    Module.Index0Float2Write=true
    Module.Index0Float3Write=true
    Module.Index1Float0Write=true
    Module.Index1Float1Write=true
    Module.Index1Float2Write=true
    Module.Index1Float3Write=true
    Write Parameter Index 1=::NewEnumerator1
   )
   &ScaleColor(
    Scale Alpha=&Multiply_Float(
     A=&FNTimeOfDayScalar(Night Scale=0.7, Morning Scale=0.9, Evening Scale=0.8)
     B=&Multiply_Float(
      A=&FN_FrontEndFloatModifier(Gameplay=0.62, Front End=0.3)
      B=&Add_Float(
       A=&Multiply_Float(
        A=&Add_Float(
         A=&Lerp_Float(A=0.15, B=0.3, Alpha=$System.AudioReactivity)
         B=&Lerp_Float(A=0, B=0.05, Alpha=$System.AudioReactivity2)
        )
        B=&MergeHelper_FloatSwitch(
         0=&FloatFromCurve(FloatCurve=DataInterface() {propertyValues {Curve {keys [<at 0.2 0 RCIM_Cubic>,<at 0.7 1>]}}}, CurveIndex=$Engine.System.Age)
         1=&FloatFromCurve(FloatCurve=DataInterface() {propertyValues {Curve {keys [<at 0.1 0 RCIM_Cubic>,<at 0.6 1>]}}}, CurveIndex=$Engine.System.Age)
         2=&FloatFromCurve(FloatCurve=DataInterface() {propertyValues {Curve {keys [<at 0 0 RCIM_Cubic>,<at 0.5 1>]}}}, CurveIndex=$Engine.System.Age)
         3=?
         Index=&SpawnGroup()
         Num=::NewEnumerator1
        )
       )
       B=&Lerp_Float(A=0, B=0.05, Alpha=$System.FN_Contrail_Setup.Dive Value)
      )
     )
    )
    Scale RGB=&VectorFromFloat(Value=&FN_FrontEndFloatModifier(Gameplay=1.2, Front End=1.5))
    ScaleA=true
    ScaleRGB=true
   )
  ]
  rendererClasses [&NiagaraMeshRendererProperties]
  renderers [{rendererClass=&NiagaraMeshRendererProperties, rendererIndex=0}]
 }
 ThinRibbon {
  emitterUpdateScript [
   &EmitterState(
    Inactive Response=::NewEnumerator1
    Life Cycle Mode=::NewEnumerator1
    Loop Behavior=::NewEnumerator1
    Loop Duration Mode=::NewEnumerator1
    UseLoopDelay=false
   )
   &SpawnBurst_Instantaneous(
    Spawn Count=&Multiply_Int(A=$System.ChainNumber, B=$System.ParticlePerRibbon)
    Use Spawn Probability=false
    UseLoopCountLimit=false
   )
  ]
  particleSpawnScript [
   &PartitionParticles(
    Number Of Partitions=$System.ChainNumber
    Particles Per Partition:NiagaraInt32=$System.ParticlePerRibbon
    Partition Mode:ENiagara_PartitionMode_v2=::NewEnumerator1
   )
   &InitializeParticle(
    Color=<rgba 0.4081 0.0586 0.4688 1>
    Lifetime=5
    Mesh Scale Mode=::NewEnumerator4
    Position Offset=<0 0 100>
    Sprite Size Mode=::NewEnumerator3
    UsePositionOffset=false
    Write Material Random=true
   )
   /SetVariables(Particles.NormalizedChainPosition:NiagaraFloat=$Particles.PartitionParticles.Normalized Index In Partition) {isSetParameters=true}
   &SkeletalMeshLocation(
    Bone / Socket Index=&ModuloInt(A=$Particles.PartitionParticles.Partition, B=4)
    Bone Sampling Mode=::NewEnumerator4
    MaskSpawnGroup=false
    Sampled Position Offset=&Lerp_Vector(
     A=&SelectVectorFromArray(
      Direct Array Index=&ModuloInt(A=$Particles.PartitionParticles.Partition, B=4)
      Vector Selection Array=DataInterface() {propertyValues {FloatData [<0 -5 0>, <0 5 0>, <0 0 0>, <0 0 0>]}}
     )
     B=&SelectVectorFromArray(
      Direct Array Index=&ModuloInt(A=$Particles.PartitionParticles.Partition, B=4)
      Vector Selection Array=DataInterface() {propertyValues {FloatData [<0 -5 150>, <0 5 150>, <0 0 75>, <0 0 75>]}}
     )
     Alpha=$Particles.NormalizedChainPosition
    )
    Sampled Position Offset Coordinate Space:ENiagaraSkelMeshTransforms=::NewEnumerator1
    Skeletal Mesh=$System.Skeletal Mesh
   )
   /SetVariables(Particles.RibbonSide:NiagaraInt32=&SelectIntFromArray(Direct Array Index=$Particles.PartitionParticles.Partition, Int Selection Array=DataInterface() {propertyValues {IntData [0,0,0,0,1,1,1,1]}})) {isSetParameters=true}
   /SetVariables(Particles.DarkRibbonCamOff=-3) {isSetParameters=true}
   /SetVariables() {isSetParameters=true}
  ]
  particleUpdateScript [
   &ParticleState(Kill Particles When Lifetime Has Elapsed=false)
   /SetVariables(Particles.CustomSpeed:NiagaraFloat=&Add_Float(
     A=$Particles.CustomSpeed
     B=&Multiply_Float(
      A=&Add_Float(A=1, B=&Multiply_Float(A=1, B=$User.ParaGlideLeanAlpha))
      B=$Engine.DeltaTime
     )
    )) {isSetParameters=true}
   &SkeletalMeshLocation(
    Bone / Socket Index=&ModuloInt(A=$Particles.PartitionParticles.Partition, B=2)
    Bone Sampling Mode=::NewEnumerator4
    MaskSpawnGroup=false
    Position Sampling=::NewEnumerator2
    Sampled Position Offset=&SelectVectorFromArray(
     Direct Array Index=$Particles.PartitionParticles.Partition
     Vector Selection Array=DataInterface() {propertyValues {FloatData [<8 -5 0>, <-8 5 0>]}}
    )
    Skeletal Mesh=$System.Skeletal Mesh
   ) {module=SkeletalMeshLocation001}
   &PendulumSetup(
    Module.Enable Spring Driver=true
    Pendulum Length:NiagaraFloat=&Lerp_Float(
     A=0
     B=&self:ScratchDynamicInput_02(
      LongLength:NiagaraFloat=$System.ChainLength
      Partition:NiagaraInt32=$Particles.PartitionParticles.Partition
      ShortLength:NiagaraFloat=$System.ChainLengthSmall
     )
     Alpha=&Lerp_Float(
      A=&FloatFromCurve(
       FloatCurve=DataInterface() {propertyValues {Curve {keys [{tangentMode=RCTM_User, time=0, value=0}, {tangentMode=RCTM_User, time=1, value=1}]}}}
       CurveIndex=$Particles.NormalizedChainPosition
       Scale Curve=&ClampFloat(Float=$Emitter.Age)
      )
      B=&FloatFromCurve(
       FloatCurve=DataInterface() {propertyValues {Curve {keys [
           {interpMode=RCIM_Cubic, leaveTangent=1.3372, tangentMode=RCTM_Break, time=0, value=0}
           {arriveTangent=0.3346, interpMode=RCIM_Cubic, leaveTangent=0.3346, tangentMode=RCTM_User, time=1, value=1}
          ]}}}
       CurveIndex=$Particles.NormalizedChainPosition
       Scale Curve=&ClampFloat(Float=$Emitter.Age)
      )
      Alpha=&FloatFromCurve(FloatCurve=DataInterface() {propertyValues {Curve {keys [<at 0 0>,<at 0.15 1>]}}}, CurveIndex=$User.ParaGlideLeanAlpha)
     )
    )
    Pendulum Pivot:Vector3f=&AddVectorToPosition(
     Position=&LerpPosition(
      A=&AddVectorToPosition(
       Position=$Particles.SkeletalMeshLocation001.SampledPosition
       Vector=&Multiply_VectorByFloat(
        Vector=&Multiply_VectorByFloat(
         Vector=&TransformVector(
          Vector=&FN_FrontEndVector3Modifier(
           Gameplay=&Lerp_Vector(
            A=&SelectVectorFromArray(
             Direct Array Index=&ModuloInt(A=$Particles.PartitionParticles.Partition, B=2)
             Vector Selection Array=$System.IdleOffsetArray
            )
            B=&SelectVectorFromArray(
             Direct Array Index=&ModuloInt(A=$Particles.PartitionParticles.Partition, B=2)
             Vector Selection Array=$System.DiveOffsetArray
            )
            Alpha=$System.SmoothDiveAlpha
           )
           Front End=&SelectVectorFromArray(
            Direct Array Index=&ModuloInt(A=$Particles.PartitionParticles.Partition, B=2)
            Vector Selection Array=$System.IdleOffsetArray
           )
          )
          Destination Space=::Simulation
         )
         Float=&SelectFloatFromArray(
          Direct Array Index=&ModuloInt(A=$Particles.PartitionParticles.Partition, B=2)
          Float Selection Array=DataInterface() {propertyValues {FloatData [1,1,0.5,0.5]}}
         )
        )
        Float=&Lerp_Float(
         A=&FloatFromCurve(
          FloatCurve=DataInterface() {propertyValues {Curve {keys [{tangentMode=RCTM_User, time=0, value=0}, {tangentMode=RCTM_User, time=1, value=1}]}}}
          CurveIndex=$Particles.NormalizedChainPosition
         )
         B=&FloatFromCurve(
          FloatCurve=DataInterface() {propertyValues {Curve {keys [{tangentMode=RCTM_User, time=0, value=0}, {tangentMode=RCTM_User, time=1, value=1}]}}}
          CurveIndex=$Particles.NormalizedChainPosition
         )
         Alpha=&FloatFromCurve(FloatCurve=DataInterface() {propertyValues {Curve {keys [<at 0 0>,<at 0.1 1>]}}}, CurveIndex=$User.ParaGlideLeanAlpha)
        )
       )
      )
      B=&AddVectorToPosition(
       Position=$Particles.SkeletalMeshLocation001.SampledPosition
       Vector=&Multiply_VectorByFloat(
        Vector=&Multiply_VectorByFloat(
         Vector=&TransformVector(
          Vector=&FN_FrontEndVector3Modifier(
           Gameplay=&Lerp_Vector(
            A=&SelectVectorFromArray(
             Direct Array Index=&ModuloInt(A=$Particles.PartitionParticles.Partition, B=4)
             Vector Selection Array=$System.IdleOffsetArray
            )
            B=&SelectVectorFromArray(
             Direct Array Index=&ModuloInt(A=$Particles.PartitionParticles.Partition, B=4)
             Vector Selection Array=$System.DiveOffsetArray
            )
            Alpha=$System.SmoothDiveAlpha
           )
           Front End=&SelectVectorFromArray(
            Direct Array Index=&ModuloInt(A=$Particles.PartitionParticles.Partition, B=4)
            Vector Selection Array=$System.IdleOffsetArray
           )
          )
          Destination Space=::Simulation
         )
         Float=&SelectFloatFromArray(
          Direct Array Index=&ModuloInt(A=$Particles.PartitionParticles.Partition, B=4)
          Float Selection Array=DataInterface() {propertyValues {FloatData [1,1,0.5,0.5]}}
         )
        )
        Float=&Lerp_Float(
         A=&FloatFromCurve(
          FloatCurve=DataInterface() {propertyValues {Curve {keys [{tangentMode=RCTM_User, time=0, value=0}, {tangentMode=RCTM_User, time=1, value=1}]}}}
          CurveIndex=$Particles.NormalizedChainPosition
         )
         B=&FloatFromCurve(
          FloatCurve=DataInterface() {propertyValues {Curve {keys [{tangentMode=RCTM_User, time=0, value=0}, {tangentMode=RCTM_User, time=1, value=1}]}}}
          CurveIndex=$Particles.NormalizedChainPosition
         )
         Alpha=&FloatFromCurve(FloatCurve=DataInterface() {propertyValues {Curve {keys [<at 0 0>,<at 0.1 1>]}}}, CurveIndex=$User.ParaGlideLeanAlpha)
        )
       )
      )
      Alpha=&MakeFloatFromInt(Int=$Particles.RibbonSide)
     )
     Vector=&TransformVector(Vector=<0 0 0>, Source Space=::World, Destination Space=::Simulation)
    )
    Pendulum Pivot Offset Coordinate Space:ENiagaraCoordinateSpace=::Local
    Pendulum Rest Axis:Vector3f=&TransformVector(
     Vector=&Lerp_Vector(
      A=&Add_Vector(
       A=&Lerp_Vector(
        A=&Add_Vector(
         A=&Multiply_VectorByFloat(
          Vector=&NormalizeVector(Vector To Normalize=$Engine.Owner.Velocity, Safe Normalize=true, Safe Fallback Threshold=0, Safe Fallback Vector=<0 0 0.001>)
          Float=-0.5
         )
         B=&TransformVector(
          Vector=&RotateVectorByQuaternion(
           Vector To Rotate=&Multiply_VectorByFloat(
            Vector=&SelectVectorFromArray(
             Direct Array Index=&ModuloInt(A=$Particles.PartitionParticles.Partition, B=2)
             Vector Selection Array=DataInterface() {propertyValues {FloatData [<0.5 1 0.1>, <0.5 1 0.1>, <-0.1 -0.25 -1>, <-0.1 -0.25 -1>]}}
            )
            Float=&SelectFloatFromArray(
             Direct Array Index=&ModuloInt(A=$Particles.PartitionParticles.Partition, B=2)
             Float Selection Array=DataInterface() {propertyValues {FloatData [1,-1,1,-1]}}
            )
           )
           Quaternion=$Particles.SkeletalMeshLocation001.SampledRotation
          )
          Destination Space=::World
         )
        )
        B=&AddVectorToPosition(
         Position=&Multiply_VectorByFloat(
          Vector=&NormalizeVector(Vector To Normalize=$Engine.Owner.Velocity, Safe Normalize=true, Safe Fallback Threshold=0, Safe Fallback Vector=<0 0 0.001>)
          Float=-1
         )
         Vector=<0 0 0.1>
        )
        Alpha=&Lerp_Float(
         A=&FloatFromCurve(
          FloatCurve=DataInterface() {propertyValues {Curve {keys [
              {arriveTangent=2.2061, interpMode=RCIM_Cubic, leaveTangent=2.2061, tangentMode=RCTM_User, time=0, value=0.2217}
              {arriveTangent=0.0794, interpMode=RCIM_Cubic, leaveTangent=0.0794, tangentMode=RCTM_User, time=1, value=1}
             ]}}}
          CurveIndex=$Particles.NormalizedChainPosition
         )
         B=1
         Alpha=&MakeCustomFloatFromBool(Bool=$User.Debug, True Float=1, False Float=&FN_FrontEndFloatModifier(Gameplay=0, Front End=1))
        )
       )
       B=&TransformVector(
        Vector=&FN_FrontEndVector3Modifier(
         Gameplay=&Lerp_Vector(
          A=&SelectVectorFromArray(
           Direct Array Index=&ModuloInt(A=$Particles.PartitionParticles.Partition, B=4)
           Vector Selection Array=DataInterface() {propertyValues {FloatData [<0 0 0>, <0 0 0>, <0 0 0>, <0 0 0>]}}
          )
          B=&SelectVectorFromArray(
           Direct Array Index=&ModuloInt(A=$Particles.PartitionParticles.Partition, B=4)
           Vector Selection Array=DataInterface() {propertyValues {FloatData [<0 0 0>, <0 0 0>, <0.25 0 0>, <-0.25 0 0>]}}
          )
          Alpha=$System.SmoothDiveAlpha
         )
         Front End=&Lerp_Vector(A=<0 -0.03 0>, B=<0 -0.15 0>, Alpha=$System.SmoothDiveAlpha)
        )
        Destination Space=::World
       )
      )
      B=&Add_Vector(
       A=&Lerp_Vector(
        A=&Add_Vector(
         A=&Multiply_VectorByFloat(
          Vector=&NormalizeVector(Vector To Normalize=$Engine.Owner.Velocity, Safe Normalize=true, Safe Fallback Threshold=0, Safe Fallback Vector=<0 0 0.001>)
          Float=-0.5
         )
         B=&TransformVector(
          Vector=&RotateVectorByQuaternion(
           Vector To Rotate=&Multiply_VectorByFloat(
            Vector=&SelectVectorFromArray(
             Direct Array Index=&ModuloInt(A=$Particles.PartitionParticles.Partition, B=4)
             Vector Selection Array=DataInterface() {propertyValues {FloatData [<0.5 1 0.1>, <0.5 1 0.1>, <-0.1 -0.25 -1>, <-0.1 -0.25 -1>]}}
            )
            Float=&SelectFloatFromArray(
             Direct Array Index=&ModuloInt(A=$Particles.PartitionParticles.Partition, B=4)
             Float Selection Array=DataInterface() {propertyValues {FloatData [1,-1,1,-1]}}
            )
           )
           Quaternion=$Particles.SkeletalMeshLocation001.SampledRotation
          )
          Destination Space=::World
         )
        )
        B=&AddVectorToPosition(
         Position=&Multiply_VectorByFloat(
          Vector=&NormalizeVector(Vector To Normalize=$Engine.Owner.Velocity, Safe Normalize=true, Safe Fallback Threshold=0, Safe Fallback Vector=<0 0 0.001>)
          Float=-1
         )
         Vector=<0 0 0.1>
        )
        Alpha=&Lerp_Float(
         A=&FloatFromCurve(
          FloatCurve=DataInterface() {propertyValues {Curve {keys [
              {arriveTangent=2.9652, interpMode=RCIM_Cubic, leaveTangent=2.9652, tangentMode=RCTM_User, time=0, value=0}
              {arriveTangent=0.0794, interpMode=RCIM_Cubic, leaveTangent=0.0794, tangentMode=RCTM_User, time=1, value=1}
             ]}}}
          CurveIndex=$Particles.NormalizedChainPosition
         )
         B=1
         Alpha=&MakeCustomFloatFromBool(Bool=$User.Debug, True Float=1, False Float=&FN_FrontEndFloatModifier(Gameplay=0, Front End=1))
        )
       )
       B=&TransformVector(
        Vector=&FN_FrontEndVector3Modifier(
         Gameplay=&Lerp_Vector(
          A=&SelectVectorFromArray(
           Direct Array Index=&ModuloInt(A=$Particles.PartitionParticles.Partition, B=4)
           Vector Selection Array=DataInterface() {propertyValues {FloatData [<0 0 0>, <0 0 0>, <0 0 0>, <0 0 0>]}}
          )
          B=&SelectVectorFromArray(
           Direct Array Index=&ModuloInt(A=$Particles.PartitionParticles.Partition, B=4)
           Vector Selection Array=DataInterface() {propertyValues {FloatData [<0 0 0>, <0 0 0>, <0.25 0 0>, <-0.25 0 0>]}}
          )
          Alpha=$System.SmoothDiveAlpha
         )
         Front End=&Lerp_Vector(A=<0 -0.03 0>, B=<0 -0.15 0>, Alpha=$System.SmoothDiveAlpha)
        )
        Destination Space=::World
       )
      )
      Alpha=&MakeFloatFromInt(Int=$Particles.RibbonSide)
     )
     Source Space=::World
     Destination Space=::Simulation
    )
    "Spring Driven Constraint (Tightness)":NiagaraFloat=&Lerp_Float(
     A=25
     B=25
     Alpha=&FloatFromCurve(
      FloatCurve=DataInterface() {propertyValues {Curve {keys [
          {arriveTangent=0.0288, interpMode=RCIM_Cubic, leaveTangent=0.0288, tangentMode=RCTM_User, time=0, value=0}
          {arriveTangent=0.9128, interpMode=RCIM_Cubic, leaveTangent=0.9128, tangentMode=RCTM_User, time=1, value=1}
         ]}}}
      CurveIndex=$Particles.NormalizedChainPosition
     )
    )
   )
   &LinearForce(
    Coordinate Space:ENiagaraCoordinateSpace=::World
    Force=&Multiply_VectorByFloat(Vector=&Add_Vector(A=$Engine.Owner.Velocity, B=$User.DebugVelocity), Float=-5)
   )
   /SetVariables(Particles.RibbonWidth=&Multiply_Float(
     A=&FloatFromCurve(
      FloatCurve=DataInterface() {propertyValues {Curve {keys [
          {arriveTangentWeight=0.5011, interpMode=RCIM_Cubic, leaveTangentWeight=4.51, tangentMode=RCTM_User, time=0.0302, value=32.4853}
          {arriveTangent=0.1551, interpMode=RCIM_Cubic, leaveTangent=0.1551, tangentMode=RCTM_User, time=1, value=11.4253}
         ]}}}
      CurveIndex=$Particles.NormalizedChainPosition
     )
     B=&Lerp_Float(A=1, B=1.1, Alpha=$System.AudioReactivity)
    )) {isSetParameters=true}
   &SolveForcesAndVelocity(Clamp Velocity=false, Limit Acceleration=false)
   &PendulumConstraint()
   &CalculateAccurateVelocity()
   &ScaleColor(
    Scale Alpha=&Multiply_Float(
     A=&FNTimeOfDayScalar(Night Scale=0.65, Morning Scale=0.85, Evening Scale=0.75)
     B=&Multiply_Float(
      A=&FN_FrontEndFloatModifier(Gameplay=2, Front End=1.2)
      B=&Add_Float(
       A=&FloatFromCurve(
        FloatCurve=DataInterface() {propertyValues {Curve {keys [
            {interpMode=RCIM_Cubic, tangentMode=RCTM_User, time=0, value=0}
            <at 0.1 1>
            {arriveTangentWeight=0.3488, interpMode=RCIM_Cubic, leaveTangentWeight=0.0997, tangentMode=RCTM_User, time=0.376, value=1}
            {tangentMode=RCTM_User, time=0.836, value=0}
           ]}}}
        CurveIndex=$Particles.NormalizedChainPosition
        Scale Curve=&Multiply_Float(
         A=&Add_Float(
          A=&Lerp_Float(A=0.06, B=0.1, Alpha=$System.AudioReactivity)
          B=&Lerp_Float(A=0, B=0.05, Alpha=$System.AudioReactivity2)
         )
         B=&FloatFromCurve(FloatCurve=DataInterface() {propertyValues {Curve {keys [<at 0.1 0 RCIM_Cubic>,<at 0.5 1>]}}}, CurveIndex=$Engine.System.Age, Scale Curve=2)
        )
       )
       B=&Lerp_Float(A=0, B=0.05, Alpha=$System.FN_Contrail_Setup.Dive Value)
      )
     )
    )
    Scale RGB=&MakeVectorFromLinearColorRGB(Color=<rgba 1 1 1 1>)
    ScaleA=true
    ScaleRGB=true
   )
   &InterpolateOverTime(
    Rate Of Change=0.25
    Target Vector Value:Vector3f=&RotateVectorByQuaternion(Vector To Rotate=<0 0 1>, Quaternion=$Particles.SkeletalMeshLocation001.SampledRotation)
    Variable Type=::NewEnumerator2
   )
   /SetVariables(Particles.RibbonFacing=&RotateVectorByQuaternion(Vector To Rotate=<0 0.5 1>, Quaternion=$Particles.SkeletalMeshLocation.SampledRotation)) {isSetParameters=true}
   /SetVariables(Particles.DarkRibbonColor:LinearColor=&MakeLinearColor(
     R:NiagaraFloat=0
     G:NiagaraFloat=0
     B:NiagaraFloat=0
     A:NiagaraFloat=&FloatFromCurve(
      FloatCurve=DataInterface() {propertyValues {Curve {keys [
          {arriveTangent=4.1728, interpMode=RCIM_Cubic, leaveTangent=4.1728, tangentMode=RCTM_User, time=0, value=0}
          {arriveTangent=-0.2338, interpMode=RCIM_Cubic, leaveTangent=-0.2338, tangentMode=RCTM_User, time=0.576, value=0.8247}
          <at 0.9499 0>
         ]}}}
      CurveIndex=$Particles.NormalizedChainPosition
      Scale Curve=&Multiply_Float(
       A=&FN_FrontEndFloatModifier(Gameplay=0.658, Front End=0.4)
       B=&FloatFromCurve(FloatCurve=DataInterface() {propertyValues {Curve {keys [<at 0.1 0 RCIM_Cubic>,<at 0.5 1>]}}}, CurveIndex=$Engine.System.Age)
      )
     )
    )) {isSetParameters=true}
   /SetVariables(Particles.DarkRibbonWidth:NiagaraFloat=&Multiply_Float(
     A=&FloatFromCurve(
      FloatCurve=DataInterface() {propertyValues {Curve {keys [
          {arriveTangent=56.142, interpMode=RCIM_Cubic, leaveTangent=56.142, tangentMode=RCTM_User, time=0, value=8.8851}
          {arriveTangentWeight=0.5011, interpMode=RCIM_Cubic, leaveTangentWeight=4.51, tangentMode=RCTM_User, time=0.2356, value=32.4853}
          {arriveTangent=-55.6777, interpMode=RCIM_Cubic, leaveTangent=-55.6777, tangentMode=RCTM_User, time=1, value=0}
         ]}}}
      CurveIndex=$Particles.NormalizedChainPosition
      Scale Curve=1.602
     )
     B=&Lerp_Float(A=1, B=1.1, Alpha=$System.AudioReactivity)
    )) {isSetParameters=true}
   &DynamicMaterialParameters(
    Index 0 Param 2=0.176
    Index 1 Param 1=&MergeHelper_FloatSwitch(0=0, 1=0.4, 2=0.7, 3=?, Index=&SpawnGroup(), Num=::NewEnumerator1)
    Index 1 Param 2=&MergeHelper_FloatSwitch(0=0.6, 1=0.9, 2=0.2, 3=?, Index=&SpawnGroup(), Num=::NewEnumerator1)
    Index 1 Param 3=&Multiply_Float(
     A=&Add_Float(
      A=&Lerp_Float(A=0.15, B=0.3, Alpha=$System.AudioReactivity)
      B=&Lerp_Float(A=0, B=0.1, Alpha=$System.AudioReactivity2)
     )
     B=&FloatFromCurve(
      FloatCurve=DataInterface() {propertyValues {Curve {keys [
          {arriveTangentWeight=0.5011, interpMode=RCIM_Cubic, leaveTangentWeight=4.51, tangentMode=RCTM_User, time=0.2558, value=1}
          {arriveTangent=0.1551, interpMode=RCIM_Cubic, leaveTangent=0.1551, tangentMode=RCTM_User, time=1, value=0}
         ]}}}
      CurveIndex=$Particles.NormalizedChainPosition
      Scale Curve=0.576
     )
    )
    Index 1 Param 4=&Multiply_Float(A=$System.AudioReactivity, B=-0.05)
    Index 2 Param 1=$Particles.CustomSpeed
    Module.Index0Float0Write=true
    Module.Index0Float1Write=true
    Module.Index0Float2Write=true
    Module.Index0Float3Write=true
    Module.Index1Float0Write=true
    Module.Index1Float1Write=true
    Module.Index1Float2Write=true
    Module.Index1Float3Write=true
    Module.Index2Float0Write=true
    Module.Index2Float1Write=true
    Module.Index2Float2Write=true
    Module.Index2Float3Write=true
    Write Parameter Index 1=::NewEnumerator1
    Write Parameter Index 2=::NewEnumerator1
   )
  ]
  rendererClasses [&NiagaraRibbonRendererProperties]
  renderers [
   {rendererClass=&NiagaraRibbonRendererProperties}
   {rendererClass=&NiagaraRibbonRendererProperties, rendererIndex=1}
  ]
 }
}
summary {
 systemName=NS_Contrail_PolarityWinnField
 emitters [
  {emitterName=RingWaves, rendererClasses [&NiagaraMeshRendererProperties]}
  {emitterName=BotCricle, rendererClasses [&NiagaraMeshRendererProperties]}
  {emitterName=Hands, rendererClasses [&NiagaraMeshRendererProperties]}
  {emitterName=ThinRibbon, rendererClasses [&NiagaraRibbonRendererProperties]}
  {emitterName=Hand_Glow, rendererClasses [&NiagaraSpriteRendererProperties]}
 ]
 userVariables [
  User.bIsFrontEnd {type=NiagaraBool}
  User.bIsFrontEndPreview {type=NiagaraBool}
  User.Debug {type=NiagaraBool}
  User.DebugVelocity {type=Vector3f, description="", defaultValue {struct=Vector3f, value=<0 0 0>}}
  User.EnableFrontendFakeAudioLoop {type=NiagaraFloat, defaultValue=1, description=""}
  User.ParaGlideLeanAlpha {type=NiagaraFloat, defaultValue=0, description=""}
  User.RibbonIDOffset {type=NiagaraInt32}
 ]
}
